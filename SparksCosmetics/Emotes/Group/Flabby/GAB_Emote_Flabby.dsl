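; Gameplay ability blueprint for the Flabby group emote in SparksCosmetics.
; It overrides CanApplyHolster to return false for any holster ID, so this emote never allows a holster to be applied.
; The activate and end ability events are only disabled placeholder parent calls, so this graph adds no other logic of its own.
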(event Ability|EventActivateAbility
  (|CallParentFunction) ; This node is disabled and will not be called.\nDrag off pins to build functionality.
)

(event Ability|EventOnEndAbility
  (params (bWasCancelled :type Boolean))
  (|CallParentFunction bWasCancelled) ; This node is disabled and will not be called.\nDrag off pins to build functionality.
)

(fn CanApplyHolster
  (params (HolsterId :type Name%20%28by%20ref%29 :default "None"))
  (return false))
